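# Foley Sound Cue for the Checkmate emote, playing a chess piece being put down.
# A Random node picks one of six equally weighted Piece_Put wave variations, then feeds a Modulator at default settings for slight pitch and volume variation.
# Despite being foley, it uses emote music class and concurrency, emote attenuation, play-when-silent virtualization and a send to the music audio-analysis submix.
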SoundCue /Game/Athena/Sounds/Emotes/Checkmate/EmoteFoley_Checkmate_Piece_Put_Cue
root=Modulator_0
nodes [
 {id=Modulator_0, children [Random_0]}
 {id=Random_0, children [WavePlayer_2,WavePlayer_3,WavePlayer_4,WavePlayer_5,WavePlayer_6,WavePlayer_7], properties {weights [1,1,1,1,1,1]}}
 {id=WavePlayer_2, properties {soundWaveAssetPtr=&/Emote_Checkmate_Piece_Put_01}}
 {id=WavePlayer_3, properties {soundWaveAssetPtr=&/Emote_Checkmate_Piece_Put_03}}
 {id=WavePlayer_4, properties {soundWaveAssetPtr=&/Emote_Checkmate_Piece_Put_05}}
 {id=WavePlayer_5, properties {soundWaveAssetPtr=&/Emote_Checkmate_Piece_Put_06}}
 {id=WavePlayer_6, properties {soundWaveAssetPtr=&/Emote_Checkmate_Piece_Put_02}}
 {id=WavePlayer_7, properties {soundWaveAssetPtr=&/Emote_Checkmate_Piece_Put_04}}
]
properties {
 attenuationSettings=&Fort_Emotes_Athena_Att
 maxDistance=1600
 soundClassObject=&Fort_Music_Emote
 virtualizationMode=PlayWhenSilent
 volumeMultiplier=1
 concurrencyOverrides {maxCount=3}
 concurrencySet [/Game/Sounds/Concurrency_Rules/Athena_EmoteMusic_Conc]
 soundSubmixSends [{minSendLevel=0, soundSubmix=&Fortnite_AudioAnalysis_Music}]
}
